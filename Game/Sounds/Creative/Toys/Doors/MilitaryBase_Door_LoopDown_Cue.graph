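# Sound cue that plays the looping movement sound of a Creative military base door, named for the downward direction.
# A fixed short delay leads into a single looping wave player for the door loop sample.
# The cue plays at reduced volume in the doors sound class, using military base door attenuation and a concurrency override that prevents new instances once the limit is reached.
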SoundCue /Game/Sounds/Creative/Toys/Doors/MilitaryBase_Door_LoopDown_Cue
root=Delay_1
nodes [
 {id=Delay_1, children [WavePlayer_0], properties {delayMax=0.1, delayMin=0.1}}
 {id=WavePlayer_0, properties {bLooping=true, soundWaveAssetPtr=&/MilitaryBase_Door_LoopDown01}}
]
properties {
 attenuationSettings=/Game/Sounds/Creative/Attenuations/Fort_MilitaryBase_Doors_Att
 maxDistance=5800
 soundClassObject=&Fort_Doors
 volumeMultiplier=0.5
 concurrencyOverrides {resolutionRule=PreventNew}
}
