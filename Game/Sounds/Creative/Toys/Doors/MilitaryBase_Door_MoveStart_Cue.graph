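# Sound cue for the start-of-movement sound of a Creative military base door.
# A single wave player feeds a modulator at the root, playing one MoveStart sample at reduced volume.
# It uses the military base door attenuation and the Fort_Doors sound class.
# A shared "only one" door concurrency asset is assigned, alongside concurrency overrides set to prevent new instances.
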SoundCue /Game/Sounds/Creative/Toys/Doors/MilitaryBase_Door_MoveStart_Cue
root=Modulator_0
nodes [
 {id=Modulator_0, children [WavePlayer_0]}
 {id=WavePlayer_0, properties {soundWaveAssetPtr=&/MilitaryBase_Door_MoveStart01}}
]
properties {
 attenuationSettings=/Game/Sounds/Creative/Attenuations/Fort_MilitaryBase_Doors_Att
 maxDistance=5800
 soundClassObject=&Fort_Doors
 volumeMultiplier=0.5
 concurrencyOverrides {resolutionRule=PreventNew}
 concurrencySet [/Game/Sounds/Creative/Concurrency/MilitaryDoor_OnlyOne_Conc]
}
