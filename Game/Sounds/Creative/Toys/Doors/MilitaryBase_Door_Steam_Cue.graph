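# Sound cue for the steam sound of a Creative military base door toy.
# A single wave player feeds the MilitaryBase_Door_Steam01 wave straight to the output, with no randomization or mixing.
# It plays at reduced volume through the Fort_Doors sound class with military base door attenuation.
# Its concurrency override sets the resolution rule to prevent new instances.
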SoundCue /Game/Sounds/Creative/Toys/Doors/MilitaryBase_Door_Steam_Cue
root=WavePlayer_1
nodes [{id=WavePlayer_1, properties {soundWaveAssetPtr=&/MilitaryBase_Door_Steam01}}]
properties {
 attenuationSettings=/Game/Sounds/Creative/Attenuations/Fort_MilitaryBase_Doors_Att
 maxDistance=5800
 soundClassObject=&Fort_Doors
 volumeMultiplier=0.5
 concurrencyOverrides {resolutionRule=PreventNew}
}
