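# Sound cue for the military base door stop sound in Creative mode.
# A mixer layers the stop wave at a fixed half pitch with the door start wave, delayed slightly by a constant offset.
# It plays quietly at high priority through the doors sound class and military door attenuation, with a concurrency set named for allowing only one door sound at a time.
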SoundCue /Game/Sounds/Creative/Toys/Doors/MilitaryBase_Door_Stop_Cue
root=Mixer_0
nodes [
 {id=Delay_0, children [WavePlayer_3], properties {delayMax=0.08, delayMin=0.08}}
 {id=Mixer_0, children [Modulator_0,Delay_0], properties {inputVolume [1,1]}}
 {id=Modulator_0, children [WavePlayer_1], properties {pitchMax=0.5, pitchMin=0.5}}
 {id=WavePlayer_1, properties {soundWaveAssetPtr=&/MilitaryBase_Door_Stop01}}
 {id=WavePlayer_3, properties {soundWaveAssetPtr=&/MilitaryBase_Door_Start01}}
]
properties {
 attenuationSettings=/Game/Sounds/Creative/Attenuations/Fort_MilitaryBase_Doors_Att
 maxDistance=5800
 priority=0.9
 soundClassObject=&Fort_Doors
 volumeMultiplier=0.3
 concurrencyOverrides {resolutionRule=PreventNew}
 concurrencySet [/Game/Sounds/Creative/Concurrency/MilitaryDoor_OnlyOne_Conc]
}
